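# Anim Blueprint for the MagmaBreak pickaxe that drives additive motion on its two large and two small floating pieces.
# Each update sets Reactivity from melee, falling or jumping state and builds real-time sine offset vectors per piece, in normal and mostly faster, wider melee sets.
# The AnimGraph blends by Reactivity between two four-bone Modify Bone chains from the cached input pose.
# The graph shows no link from those vectors to the bone translations.
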AnimBlueprint /BRCosmetics/Weapons/FORT_Melee/Pickaxe_MagmaBreak/Meshes/MagmaBreak_Axe_AnimBP
target_skeleton=&/MagmaBreak_Axe_Skeleton
graphs [
 AnimGraph {nodes [
   n1 = "Transform (Modify) Bone - Bone: large_piece_l"(
    ComponentPose=Local To Component(LocalPose=Use cached pose 'Input cache'())
    Alpha=0.85
    Translation="0,0,0"
    Rotation="0,0,0"
    Scale="1,1,1"
   ) {translationMode=BMM_Additive, translationSpace=BCS_BoneSpace, alphaBoolBlend {blendOption=Linear}, boneToModify {boneName=large_piece_l}}
   n2 = "Transform (Modify) Bone - Bone: large_piece_r"(ComponentPose=@n1, Alpha=0.85, Translation="0,0,0", Rotation="0,0,0", Scale="1,1,1") {translationMode=BMM_Additive, translationSpace=BCS_BoneSpace, alphaBoolBlend {blendOption=Linear}, boneToModify {boneName=large_piece_r}}
   n3 = "Transform (Modify) Bone - Bone: small_piece_l"(ComponentPose=@n2, Alpha=0.85, Translation="0,0,0", Rotation="0,0,0", Scale="1,1,1") {translationMode=BMM_Additive, translationSpace=BCS_BoneSpace, alphaBoolBlend {blendOption=Linear}, boneToModify {boneName=small_piece_l}}
   n4 = "Transform (Modify) Bone - Bone: small_piece_r"(ComponentPose=@n3, Alpha=0.85, Translation="0,0,0", Rotation="0,0,0", Scale="1,1,1") {translationMode=BMM_Additive, translationSpace=BCS_BoneSpace, alphaBoolBlend {blendOption=Linear}, boneToModify {boneName=small_piece_r}}
   n5 = "Transform (Modify) Bone - Bone: large_piece_l"!(ComponentPose=Local To Component(LocalPose=Use cached pose 'Input cache'()), Translation="0,0,0", Rotation="0,0,0", Scale="1,1,1") -> (Pose) {
    translationMode=BMM_Additive
    translationSpace=BCS_BoneSpace
    alphaBoolBlend {blendInTime=0.1, blendOption=Linear, blendOutTime=3}
    alphaScaleBiasClamp {bInterpResult=true, interpSpeedDecreasing=1.1, outRange {max=0.7}}
    boneToModify {boneName=large_piece_l}
   }
   n6 = "Transform (Modify) Bone - Bone: large_piece_r"!(ComponentPose=@n5, Translation="0,0,0", Rotation="0,0,0", Scale="1,1,1") -> (Pose) {
    translationMode=BMM_Additive
    translationSpace=BCS_BoneSpace
    alphaBoolBlend {blendInTime=0.1, blendOption=Linear, blendOutTime=3}
    alphaScaleBiasClamp {bInterpResult=true, interpSpeedDecreasing=1.1, outRange {max=0.7}}
    boneToModify {boneName=large_piece_r}
   }
   n7 = "Transform (Modify) Bone - Bone: small_piece_l"!(ComponentPose=@n6, Translation="0,0,0", Rotation="0,0,0", Scale="1,1,1") -> (Pose) {
    translationMode=BMM_Additive
    translationSpace=BCS_BoneSpace
    alphaBoolBlend {blendInTime=0.1, blendOption=Linear, blendOutTime=3}
    alphaScaleBiasClamp {bInterpResult=true, interpSpeedDecreasing=1.1, outRange {max=0.7}}
    boneToModify {boneName=small_piece_l}
   }
   n8 = "Transform (Modify) Bone - Bone: small_piece_r"!(ComponentPose=@n7, Translation="0,0,0", Rotation="0,0,0", Scale="1,1,1") -> (Pose) {
    translationMode=BMM_Additive
    translationSpace=BCS_BoneSpace
    alphaBoolBlend {blendInTime=0.1, blendOption=Linear, blendOutTime=3}
    alphaScaleBiasClamp {bInterpResult=true, interpSpeedDecreasing=1.1, outRange {max=0.7}}
    boneToModify {boneName=small_piece_r}
   }
   n9 = Blend Poses by bool(
    BlendPose_0=Component To Local(ComponentPose=@n8)
    BlendPose_1=Component To Local(ComponentPose=@n4)
    BlendTime_1=2.3
    bActiveValue=Get Reactivity() -> (Reactivity:bool)
   ) {blendType=Linear}
   Output Pose(Result=Component To Local(ComponentPose=Local To Component(LocalPose=@n9)))
   Save cached pose 'Input cache'(Pose=Input Pose()) {cacheName=Input cache}
  ]}
 EventGraph {nodes [
   Event BlueprintUpdateAnimation()
   > n1 = Sequence!(execute) -> (then_0, then_1, then_2:Exec)
   > Set Reactivity(Reactivity:bool=OR Boolean!(A=Get bIsPlayingMeleeAnim(), B=Get bIsFalling(), C:bool=Get bIsJumping()) -> (ReturnValue)) -> (Output_Get:bool)
   n2 = GetRealTimeSeconds()
   n3 = "Sin (Radians)"(A=float * float(A=1, B=@n2))
   n4 = "Sin (Radians)"(A=float * float(A=3.9, B=@n2))
   Set dyn_Large_Left_Float(execute=@n1.then_1, dyn_Large_Left_Float:Vector=MakeVector(X=float * float(A=@n4, B=0.8), Y=0, Z=float * float(A=@n3, B=0.3))) -> (Output_Get:Vector)
   n5 = GetRealTimeSeconds()
   n6 = "Sin (Radians)"(A=float * float(A=-1, B=@n5))
   n7 = "Sin (Radians)"(A=float * float(A=-3.92, B=@n5))
   > Set dyn_Large_Right_Float(dyn_Large_Right_Float:Vector=MakeVector(X=float * float(A=@n7, B=0.8), Y=0, Z=float * float(A=@n6, B=0.3))) -> (Output_Get:Vector)
   n8 = GetRealTimeSeconds()
   n9 = "Sin (Radians)"(A=float * float(A=1.1, B=@n8))
   n10 = "Sin (Radians)"(A=float * float(A=2.52, B=@n8))
   > Set dyn_Small_Left_Float(dyn_Small_Left_Float:Vector=MakeVector(X=float * float(A=@n10, B=1.35), Y=0, Z=float * float(A=@n9, B=-0.5))) -> (Output_Get:Vector)
   n11 = GetRealTimeSeconds()
   n12 = "Sin (Radians)"(A=float * float(A=-1.1, B=@n11))
   n13 = "Sin (Radians)"(A=float * float(A=-2.51, B=@n11))
   > Set dyn_Small_Righ_Float(dyn_Small_Righ_Float:Vector=MakeVector(X=float * float(A=@n13, B=1.35), Y=0, Z=float * float(A=@n12, B=-0.5))) -> (Output_Get:Vector)
   n14 = GetRealTimeSeconds()
   n15 = "Sin (Radians)"(A=float * float(A=5, B=@n14))
   n16 = "Sin (Radians)"(A=float * float(A=2, B=@n14))
   Set dyn_Large_Left_Float_Melee(execute=@n1.then_2, dyn_Large_Left_Float_Melee:Vector=MakeVector(X=float * float(A=@n15, B=1.2), Y=0, Z=float * float(A=@n16, B=0.6))) -> (Output_Get:Vector)
   n17 = GetRealTimeSeconds()
   n18 = "Sin (Radians)"(A=float * float(A=-2, B=@n17))
   n19 = "Sin (Radians)"(A=float * float(A=-5, B=@n17))
   > Set dyn_Large_Right_Float_Melee(dyn_Large_Right_Float_Melee:Vector=MakeVector(X=float * float(A=@n19, B=1.2), Y=0, Z=float * float(A=@n18, B=0.6))) -> (Output_Get:Vector)
   n20 = GetRealTimeSeconds()
   n21 = "Sin (Radians)"(A=float * float(A=2, B=@n20))
   n22 = "Sin (Radians)"(A=float * float(A=5, B=@n20))
   > Set dyn_Small_Left_Float_Melee(dyn_Small_Left_Float_Melee:Vector=MakeVector(X=float * float(A=@n22, B=1.8), Y=0, Z=float * float(A=@n21, B=-1))) -> (Output_Get:Vector)
   n23 = GetRealTimeSeconds()
   n24 = "Sin (Radians)"(A=float * float(A=-2, B=@n23))
   n25 = "Sin (Radians)"(A=float * float(A=-5, B=@n23))
   > Set dyn_Small_Righ_Float_Melee(dyn_Small_Righ_Float_Melee:Vector=MakeVector(X=float * float(A=@n25, B=1.8), Y=0, Z=float * float(A=@n24, B=-1))) -> (Output_Get:Vector)
  ]}
]
variables [
 dyn_Large_Left_Float {type=Vector}
 dyn_Large_Left_Float_Melee {type=Vector}
 dyn_Large_Right_Float {type=Vector}
 dyn_Large_Right_Float_Melee {type=Vector}
 dyn_Small_Left_Float {type=Vector}
 dyn_Small_Left_Float_Melee {type=Vector}
 dyn_Small_Righ_Float {type=Vector}
 dyn_Small_Righ_Float_Melee {type=Vector}
 Reactivity {type=boolean}
]
