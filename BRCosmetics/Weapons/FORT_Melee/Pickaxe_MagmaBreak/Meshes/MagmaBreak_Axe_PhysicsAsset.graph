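# Physics asset for the MagmaBreak pickaxe cosmetic's axe skeletal mesh, a two-handed melee harvesting weapon.
# It defines a single capsule body on the Melee_TwoHanded_Handle bone, offset and rotated from the bone origin.
# It has no constraints, so the axe is represented as one collision shape with no jointed parts.
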PhysicsAsset /BRCosmetics/Weapons/FORT_Melee/Pickaxe_MagmaBreak/Meshes/MagmaBreak_Axe_PhysicsAsset
bodies {Melee_TwoHanded_Handle [{shapeType=Capsule, length=35.6671, radius=12.9648, center=<9.2982 -14.5809 0.2168>, rotation=<rot 0 4.5761 -88.4814>}]}
constraints []
